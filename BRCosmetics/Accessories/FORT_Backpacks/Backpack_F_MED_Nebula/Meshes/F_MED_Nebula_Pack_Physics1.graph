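# Physics asset for the Nebula backpack cosmetic, letting its fifteen leaf bones move as simulated bodies.
# Each Dyn_Leaf bone has a capsule body, and a tiny capsule on the pack bone serves as the shared anchor.
# Every leaf is constrained directly to the pack rather than chained, with near-locked twist, narrow first swing and a wider, per-leaf second swing, so leaves flap mostly along one axis.
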PhysicsAsset /BRCosmetics/Accessories/FORT_Backpacks/Backpack_F_MED_Nebula/Meshes/F_MED_Nebula_Pack_Physics1
bodies {
 Dyn_Leaf_01 [{shapeType=Capsule, length=3.6562, radius=2.4688, shapeName=Dyn_Leaf_01_capsule, center=<2 0 0>, rotation=<rot 90 0 0>}]
 Dyn_Leaf_02 [{shapeType=Capsule, length=3.125, radius=1.5, shapeName=Dyn_Leaf_02_capsule, center=<5 0 0>, rotation=<rot 90 0 0>}]
 Dyn_Leaf_03 [{shapeType=Capsule, length=1.8166, radius=1.0193, shapeName=Dyn_Leaf_03_capsule, center=<2 0 0>, rotation=<rot 90 0 0>}]
 Dyn_Leaf_04 [{
   shapeType=Capsule
   length=5.3438
   radius=2.9375
   shapeName=Dyn_Leaf_04_capsule
   center=<6.4413 0 1.6165>
   rotation=<rot 70.0001 179.9997 179.9998>
  }]
 Dyn_Leaf_05 [{shapeType=Capsule, length=2.4688, radius=1.9062, shapeName=Dyn_Leaf_05_capsule, center=<2 0 0>, rotation=<rot 90 0 0>}]
 Dyn_Leaf_06 [{shapeType=Capsule, length=2.4688, radius=1.9062, shapeName=Dyn_Leaf_06_capsule, center=<2 0 0>, rotation=<rot 90 0 0>}]
 Dyn_Leaf_07 [{shapeType=Capsule, length=1.8125, radius=2.3125, shapeName=Dyn_Leaf_07_capsule, center=<1 0 0>, rotation=<rot 90 0 0>}]
 Dyn_Leaf_08 [{shapeType=Capsule, length=2.5625, radius=2.9688, shapeName=Dyn_Leaf_08_capsule, center=<2.5 0 0>, rotation=<rot 90 0 0>}]
 Dyn_Leaf_09 [{shapeType=Capsule, length=2, radius=3, shapeName=Dyn_Leaf_09_capsule, center=<1 0 -1>, rotation=<rot 110 0 0>}]
 Dyn_Leaf_10 [{shapeType=Capsule, length=3.0312, radius=2.375, shapeName=Dyn_Leaf_10_capsule, center=<2.5 0 0>, rotation=<rot 90 0 0>}]
 Dyn_Leaf_11 [{shapeType=Capsule, length=2.5312, radius=2.125, shapeName=Dyn_Leaf_11_capsule, center=<2 0 0>, rotation=<rot 90 0 0>}]
 Dyn_Leaf_12 [{shapeType=Capsule, length=1.9688, radius=2.0625, shapeName=Dyn_Leaf_12_capsule, center=<2 0 0>, rotation=<rot 90 0 0>}]
 Dyn_Leaf_13 [{shapeType=Capsule, length=1.75, radius=1.9688, shapeName=Dyn_Leaf_13_capsule, center=<2 0 0>, rotation=<rot 90 0 0>}]
 Dyn_Leaf_14 [{shapeType=Capsule, length=1.75, radius=1.875, shapeName=Dyn_Leaf_14_capsule, center=<2 0 0>, rotation=<rot 85 -179.9994 -179.9994>}]
 Dyn_Leaf_15 [{shapeType=Capsule, length=1.9062, radius=2, shapeName=Dyn_Leaf_15_capsule, center=<2 0 0>, rotation=<rot 90 0 0>}]
 pack [{shapeType=Capsule, length=0.505, radius=0.505, shapeName=pack_capsule, rotation=<rot 0 -90 0>}]
}
constraints [
 {bone1Name=Dyn_Leaf_01, bone2Name=pack, swing1LimitDegrees=1, swing2LimitDegrees=15, twistLimitDegrees=1}
 {bone1Name=Dyn_Leaf_02, bone2Name=pack, swing1LimitDegrees=0, swing2LimitDegrees=5, twistLimitDegrees=1}
 {bone1Name=Dyn_Leaf_03, bone2Name=pack, swing1LimitDegrees=3, swing2LimitDegrees=45, twistLimitDegrees=1}
 {bone1Name=Dyn_Leaf_04, bone2Name=pack, swing1LimitDegrees=3, swing2LimitDegrees=15, twistLimitDegrees=1}
 {bone1Name=Dyn_Leaf_05, bone2Name=pack, swing1LimitDegrees=0, swing2LimitDegrees=10, twistLimitDegrees=1}
 {bone1Name=Dyn_Leaf_06, bone2Name=pack, swing1LimitDegrees=1, swing2LimitDegrees=15, twistLimitDegrees=1}
 {bone1Name=Dyn_Leaf_07, bone2Name=pack, swing1LimitDegrees=5, swing2LimitDegrees=40, twistLimitDegrees=1}
 {bone1Name=Dyn_Leaf_08, bone2Name=pack, swing1LimitDegrees=5, swing2LimitDegrees=25, twistLimitDegrees=1}
 {bone1Name=Dyn_Leaf_09, bone2Name=pack, swing1LimitDegrees=1, swing2LimitDegrees=20, twistLimitDegrees=1}
 {bone1Name=Dyn_Leaf_10, bone2Name=pack, swing1LimitDegrees=8, swing2LimitDegrees=40, twistLimitDegrees=1}
 {bone1Name=Dyn_Leaf_11, bone2Name=pack, swing1LimitDegrees=1, swing2LimitDegrees=25, twistLimitDegrees=1}
 {bone1Name=Dyn_Leaf_12, bone2Name=pack, swing1LimitDegrees=6, swing2LimitDegrees=30, twistLimitDegrees=1}
 {bone1Name=Dyn_Leaf_13, bone2Name=pack, swing1LimitDegrees=5, swing2LimitDegrees=20, twistLimitDegrees=1}
 {bone1Name=Dyn_Leaf_14, bone2Name=pack, swing1LimitDegrees=3, swing2LimitDegrees=20, twistLimitDegrees=1}
 {bone1Name=Dyn_Leaf_15, bone2Name=pack, swing1LimitDegrees=5, swing2LimitDegrees=25, twistLimitDegrees=1}
]
