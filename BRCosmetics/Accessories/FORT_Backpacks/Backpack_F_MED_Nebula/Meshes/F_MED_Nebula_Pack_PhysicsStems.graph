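# Physics asset for the Nebula backpack cosmetic that defines bodies and joint limits for its four dynamic stems.
# Each stem is a two-bone chain of tiny capsules, with zero-length tip capsules acting as spheres, anchored to a small capsule on the pack bone.
# Base joints have asymmetric swing limits and tip joints symmetric ones, with twist nearly locked throughout, so stems bend slightly without spinning.
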PhysicsAsset /BRCosmetics/Accessories/FORT_Backpacks/Backpack_F_MED_Nebula/Meshes/F_MED_Nebula_Pack_PhysicsStems
bodies {
 Dyn_Stem_01_01 [{shapeType=Capsule, length=1, radius=0.5, shapeName=Dyn_Stem_01_01_capsule, center=<0.5 0 0>, rotation=<rot 90 0 0>}]
 Dyn_Stem_01_02 [{shapeType=Capsule, length=0, radius=0.5, shapeName=Dyn_Stem_01_02_capsule, center=<0.5 0 0>, rotation=<rot 90 0 0>}]
 Dyn_Stem_02_01 [{shapeType=Capsule, length=1, radius=0.5, shapeName=Dyn_Stem_02_01_capsule, center=<0.5 0 0>, rotation=<rot 90 -29.7867 -19.7868>}]
 Dyn_Stem_02_02 [{shapeType=Capsule, length=0, radius=0.5, shapeName=Dyn_Stem_02_02_capsule, center=<0.5 0 0>, rotation=<rot 90 0 0>}]
 Dyn_Stem_03_01 [{shapeType=Capsule, length=1, radius=0.5, shapeName=Dyn_Stem_03_01_capsule, center=<0.5 0 0>, rotation=<rot 69.9998 -0.0001 -4.0482>}]
 Dyn_Stem_03_02 [{shapeType=Capsule, length=0, radius=0.5, shapeName=Dyn_Stem_03_02_capsule, center=<0.5 0 0.4976>, rotation=<rot 70.0001 -179.9998 -179.9999>}]
 Dyn_Stem_04_01 [{shapeType=Capsule, length=1, radius=0.5, shapeName=Dyn_Stem_04_01_capsule, center=<0.5 0 0>, rotation=<rot 69.9998 -0.0001 -0.0002>}]
 Dyn_Stem_04_02 [{shapeType=Capsule, length=0, radius=0.5, shapeName=Dyn_Stem_04_02_capsule, center=<0.5 0 -0.2464>, rotation=<rot 90 0 0>}]
 pack [{shapeType=Capsule, length=0.505, radius=0.505, shapeName=pack_capsule, rotation=<rot 0 -90 0>}]
}
constraints [
 {bone1Name=Dyn_Stem_01_02, bone2Name=Dyn_Stem_01_01, swing1LimitDegrees=15, swing2LimitDegrees=15, twistLimitDegrees=1}
 {bone1Name=Dyn_Stem_02_02, bone2Name=Dyn_Stem_02_01, swing1LimitDegrees=15, swing2LimitDegrees=15, twistLimitDegrees=1}
 {bone1Name=Dyn_Stem_03_02, bone2Name=Dyn_Stem_03_01, swing1LimitDegrees=15, swing2LimitDegrees=15, twistLimitDegrees=1}
 {bone1Name=Dyn_Stem_04_02, bone2Name=Dyn_Stem_04_01, swing1LimitDegrees=15, swing2LimitDegrees=15, twistLimitDegrees=1}
 {bone1Name=Dyn_Stem_01_01, bone2Name=pack, swing1LimitDegrees=8, swing2LimitDegrees=25, twistLimitDegrees=1}
 {bone1Name=Dyn_Stem_02_01, bone2Name=pack, swing1LimitDegrees=8, swing2LimitDegrees=25, twistLimitDegrees=1}
 {bone1Name=Dyn_Stem_03_01, bone2Name=pack, swing1LimitDegrees=8, swing2LimitDegrees=25, twistLimitDegrees=1}
 {bone1Name=Dyn_Stem_04_01, bone2Name=pack, swing1LimitDegrees=8, swing2LimitDegrees=25, twistLimitDegrees=1}
]
